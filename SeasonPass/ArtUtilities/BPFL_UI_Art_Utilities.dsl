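; Season Pass UI art function library whose AssignNexusUpsellBenefitsColors function applies upsell benefit theme colors to widgets.
; For each assignment, it looks up the source color via SeasonPassVM with a magenta fallback and records it in the returned map.
; Image and Border widgets supply dynamic materials whose target vector parameter receives that color.
; An EditableTextBox class check leads to a TextBlock cast for setting text color.
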(fn AssignNexusUpsellBenefitsColors
  (params (BenefitsData :type Nexus%20Upsell%20Benefits%20Structure) (ColorAssignments :type Array%20of%20S%20UI%20Art%20Utilities%20Material%20Parameters%20Structures))
  (if (Utilities|Array|IsNotEmpty ColorAssignments)
    (bind _output_get (LocalVariables|SetColorAssignmentArray ColorAssignments))
    (for _array_element _output_get
      (LocalVariables|SetWidget (Utilities|Struct|BreakSUIArtUtilitiesMaterialParameters _array_element))
      (LocalVariables|SetSource (.SourceValue_7_C2C139B34828522DFDBA03957F34C8D4 (Utilities|Struct|BreakSUIArtUtilitiesMaterialParameters _array_element)))
      (LocalVariables|SetTarget (.TargetParameter_8_AB1F32FF467CF02BA6394FAAEF1D02C4 (Utilities|Struct|BreakSUIArtUtilitiesMaterialParameters _array_element)))
      (if (or (Utilities|String|NotEqualExactly(String) (LocalVariables|GetSource)) (Utilities|String|NotEqualExactly(String) (LocalVariables|GetSource) "None"))
        (bind _output_get_2 (LocalVariables|SetSourceColorValue (Utilities|Struct|BreakSlateColor (Class|SeasonPassVM|GetUpsellThemeColor (Variables|GetBenefitsData) (LocalVariables|GetSource) "(R=1.000000,G=0.000000,B=1.000000,A=1.000000)"))))
        (Utilities|Map|Add (LocalVariables|GetBenefitsColorMap) (LocalVariables|GetSource) _output_get_2) ; Add data Color values
        (if (or (or (!= (LocalVariables|GetTarget) "") (!= (LocalVariables|GetTarget) "")) (Utilities|IsValid (LocalVariables|GetWidget) :!class /Script/Engine.KismetSystemLibrary))
          (if (Utilities|ClassIsChildOf (Utilities|GetClass (LocalVariables|GetWidget)) "/Script/UMG.Image")
            (bind _asimage (Utilities|Casting|CastToImage (LocalVariables|GetWidget)))
            (bind _returnvalue (Appearance|GetDynamicMaterial _asimage :!class /Script/UMG.Image))
            (LocalVariables|SetDynamicMaterial _returnvalue)
            (elif (Utilities|ClassIsChildOf (Utilities|GetClass (LocalVariables|GetWidget)) "/Script/UMG.Border")
              (bind _asborder (Utilities|Casting|CastToBorder (LocalVariables|GetWidget)))
              (bind _returnvalue_2 (Appearance|GetDynamicMaterial _asborder :!class /Script/UMG.Border))
              (LocalVariables|SetDynamicMaterial _returnvalue_2)
              (else
                (if (Utilities|ClassIsChildOf (Utilities|GetClass (LocalVariables|GetWidget)) "/Script/UMG.EditableTextBox")
                  (bind _astext (Utilities|Casting|CastToTextBlock (LocalVariables|GetWidget)))
                  (Appearance|SetColorandOpacity _astext (Utilities|Struct|MakeSlateColor (LocalVariables|GetSourceColorValue)) :!class /Script/UMG.TextBlock)))))
          (Rendering|Material|SetVectorParameterValue (LocalVariables|GetDynamicMaterial) (LocalVariables|GetTarget) (LocalVariables|GetSourceColorValue) :!class /Script/Engine.MaterialInstanceDynamic))))
    (return (LocalVariables|GetBenefitsColorMap))
    (else
      (return (Utilities|Map|MakeMap "")))))
